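# Niagara idle effect for the BubbleBowl cosmetic shoes, showing four persistent fish and running ribbon trails.
# Fish uses exec index to pick each particle's bone/socket, orientation and shoe scale, with user parameters blending position offsets.
# Per-fish motion values and orientation axes reach materials through dynamic parameters.
# Trail spawns two skeletal anim-trail ribbons while running, hides them during skydiving unless debugging, and applies time-of-day tint plus shoe-scaled width.
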NiagaraSystem /CosmeticShoes/Assets/FORT_Shoes/Shoe_BubbleBowl/FX/NS_Shoe_BubbleBowl_Idle
dependencies {
 usedDataInterfaces [
  &NiagaraDataInterfaceColorCurve
  &NiagaraDataInterfaceCurve
  &NiagaraDataInterfaceArrayFloat3
  &NiagaraDataInterfaceArrayFloat
  &NiagaraDataInterfaceSkeletalMesh
 ]
 usedDynamicInputs [
  /Game/Effects/Niagara/DynamicInputs/Logic/FN_IsRunningCheck
  &Divide_Float
  &FNTimeOfDayColor
 ]
 usedModules [
  &self:SystemSpawnScript.SystemScriptSource.SystemScriptGraph.NiagaraNodeAssignment_3.SetVariables_#1
  &SystemState
  &self:SystemSpawnScript.SystemScriptSource.SystemScriptGraph.NiagaraNodeAssignment_1.SetVariables_#2
  &self:SystemSpawnScript.SystemScriptSource.SystemScriptGraph.NiagaraNodeAssignment_2.SetVariables_#3
  &self:SystemSpawnScript.SystemScriptSource.SystemScriptGraph.NiagaraNodeAssignment_0.SetVariables_#4
 ]
}
emitters {
 Fish {
  emitterUpdateScript [
   &EmitterState(Life Cycle Mode=::NewEnumerator1, Loop Behavior=::NewEnumerator1, Loop Duration=1.0333, UseLoopDelay=false)
   &SpawnBurst_Instantaneous(Spawn Count=4, Use Spawn Probability=false, UseLoopCountLimit=false)
  ]
  particleSpawnScript [
   &InitializeParticle(
    Color Mode=::NewEnumerator0
    Lifetime=5
    Mesh Scale Mode=::NewEnumerator3
    Mesh Uniform Scale=&Multiply_Float(
     A=0.85
     B=&MergeHelper_FloatSwitch(
      0=&MakeFloatFromVector(Vector=$User.ShoeScale_L, Channel=::NewEnumerator0)
      1=&MakeFloatFromVector(Vector=$User.ShoeScale_L, Channel=::NewEnumerator0)
      2=&MakeFloatFromVector(Vector=$User.ShoeScale_R, Channel=::NewEnumerator0)
      3=&MakeFloatFromVector(Vector=$User.ShoeScale_R, Channel=::NewEnumerator0)
      Index=&ReturnExecIndex()
      Num=::NewEnumerator2
     )
    )
    UsePositionOffset=false
    Write Material Random=true
   )
   &InitialMeshOrientation(
    Mesh Orientation Mode=::NewEnumerator7
    Rotation=&SelectVectorFromArray(
     Direct Array Index=&ReturnExecIndex()
     Vector Selection Array=DataInterface() {propertyValues {FloatData [<0 0.25 0>, <0.25 0 0>, <0 -0.25 0>, <-0.25 0 0>]}}
    )
    Use Rotation Vector=true
   )
  ]
  particleUpdateScript [
   &ParticleState(Kill Particles When Lifetime Has Elapsed=false)
   &SkeletalMeshLocation(
    Bone / Socket Index=&ReturnExecIndex()
    Bone Sampling Mode=::NewEnumerator4
    MaskSpawnGroup=false
    Sampled Position Offset=/Niagara/DynamicInputs/Lerp/Lerp_VectorComponents(
     Lerp Component Alpha:Vector3f=&MakeVector(X=$User.FrontFishOffsetLerpX, Y=$User.FrontFishOffsetLerpY, Z=0)
     A:Vector3f=&SelectVectorFromArray(
      Direct Array Index=&ReturnExecIndex()
      Vector Selection Array=DataInterface() {propertyValues {FloatData [<-6.2 -0.5 0>, <2 2.2 0>, <6.2 0.5 0>, <-2 -2.2 0>]}}
     )
     B:Vector3f=&Add_Vector(
      A=&SelectVectorFromArray(
       Direct Array Index=&ReturnExecIndex()
       Vector Selection Array=DataInterface() {propertyValues {FloatData [<-6.5 -1 0>, <-2 0.5 0>, <6.5 1 0>, <2 -0.5 0>]}}
      )
      B=&Multiply_Vector(
       A=&SelectVectorFromArray(
        Direct Array Index=&ReturnExecIndex()
        Vector Selection Array=DataInterface() {propertyValues {FloatData [<-1.5 0 0>, <0 0 0>, <1.5 0 0>, <0 0 0>]}}
       )
       B=&MakeVector(X=$User.RearFishOffsetLerpZ, Y=0, Z=0)
      )
     )
    )
    Skeletal Mesh=$System.Skeletal Mesh Reader
   )
   &DynamicMaterialParameters(
    Index 0 Param 1=&Multiply_Float(
     A=&SelectFloatFromArray(
      Direct Array Index=&ReturnExecIndex()
      Float Selection Array=DataInterface() {propertyValues {FloatData [3,3.7,-3,-3.7]}}
     )
     B=&MakeFloatFromVector(Vector=$User.ShoeScale_L, Channel=::NewEnumerator0)
    )
    Index 0 Param 2=&SelectFloatFromArray(
     Direct Array Index=&ReturnExecIndex()
     Float Selection Array=DataInterface() {propertyValues {FloatData [0.25,0.3,-0.25,-0.3]}}
    )
    Index 0 Param 3=&SelectFloatFromArray(
     Direct Array Index=&ReturnExecIndex()
     Float Selection Array=DataInterface() {propertyValues {FloatData [0,0,0,0]}}
    )
    Index 0 Param 4=&SelectFloatFromArray(Direct Array Index=&ReturnExecIndex(), Float Selection Array=DataInterface() {propertyValues {FloatData []}})
    Index 1 Vector 4:Vector4f=&MakeFloat4FromVectorAndFloat(
     "Vector (RGB)"=&ReturnMeshOrientationAxis(Mesh Orientation Axis=::XAxis, Mesh Orientation Quat=$Particles.MeshOrientation)
     "Float (Alpha)"=1
    )
    Index 2 Vector 4:Vector4f=&MakeFloat4FromVectorAndFloat(
     "Vector (RGB)"=&ReturnMeshOrientationAxis(Mesh Orientation Axis=::YAxis, Mesh Orientation Quat=$Particles.MeshOrientation)
     "Float (Alpha)"=1
    )
    Index 3 Vector 4:Vector4f=&MakeFloat4FromVectorAndFloat(
     "Vector (RGB)"=&ReturnMeshOrientationAxis(Mesh Orientation Axis=::ZAxis, Mesh Orientation Quat=$Particles.MeshOrientation)
     "Float (Alpha)"=1
    )
    Module.Index0Float0Write=true
    Module.Index0Float1Write=true
    Module.Index0Float2Write=true
    Module.Index0Float3Write=false
    Write Parameter Index 1=::NewEnumerator4
    Write Parameter Index 2=::NewEnumerator4
    Write Parameter Index 3=::NewEnumerator4
   )
  ]
  rendererClasses [&NiagaraSpriteRendererProperties, &NiagaraMeshRendererProperties]
  renderers [
   {rendererClass=&NiagaraSpriteRendererProperties}
   {rendererClass=&NiagaraMeshRendererProperties, rendererIndex=1}
   {rendererClass=&NiagaraMeshRendererProperties, rendererIndex=2}
  ]
 }
 Trail {
  emitterSpawnScript [&FN_RibbonIDOffsetUtility()]
  emitterUpdateScript [
   &EmitterState()
   &FN_SpawnAnimTrail(
    Number of Anim Trails:NiagaraInt32=&MakeCustomIntFromBool(Bool=&Boolean_OrOperation(Bool1=$System.bIsRunning, Bool2=$System.DEBUGSHOWTRAIL), True INT=2, False INT=0)
    Spawn Mode:ENiagaraSkeletalSpawning=::NewEnumerator0
    Switch on ENiagara_AnimTrail_Source:ENiagara_AnimTrail_Source=::NewEnumerator1
   )
  ]
  particleSpawnScript [
   &FN_InitializeAnimTrailBoneIndecies(Second Bone or Socket Index=0)
   &FN_InitializeAnimTrailBoneIndecies(AnimTrail Selection=1, First Bone or Socket Index=2, Second Bone or Socket Index=2) {module=FN_InitializeAnimTrailBoneIndecies001}
   &FN_InitializeAnimTrail(AnimTrailWidth=30, Color=<rgba 1 1 1 0.5>, Lifetime=0.15, SkeletalMesh=$System.Skeletal Mesh Reader)
   &AddVelocity(Velocity=&FN_FrontEndVector3Modifier(Gameplay=<0 0 0>, Front End=<0 -350 0>))
  ]
  particleUpdateScript [
   &SolveForcesAndVelocity(Clamp Velocity=false, Limit Acceleration=false)
   &FN_UpdateAnimTrail(SkeletalMesh=$System.Skeletal Mesh Reader) {module=FN_UpdateAnimTrail001}
   &ScaleColor(
    Linear Color Curve=DataInterface() {propertyValues {
      AlphaCurve {keys [<at 0.0144 1.2>,<at 0.0876 0.6>,<at 1 0>]}
      BlueCurve {keys [<at 0.2507 1>,<at 0.4595 0.8504>,<at 0.6162 0.5781>]}
      GreenCurve {keys [<at 0.2507 0.894>,<at 0.4595 0.8958>,<at 0.6162 0.3941>]}
      RedCurve {keys [<at 0.2507 0.6>,<at 0.4595 0.5375>,<at 0.6162 0.3469>]}
     }}
    Scale Mode=::NewEnumerator2
   )
   &DynamicMaterialParameters(
    Index 0 Param 1=&MergeHelper_FloatSwitch(0=0.5, 1=0, 2=?, 3=?, Index=&SpawnGroup(), Num=::NewEnumerator0)
    Index 0 Param 4=1
    Module.Index0Param0Write=true
    Module.Index0Param1Write=true
    Module.Index0Param2Write=true
    Module.Index0Param3Write=true
   ) {module=DynamicMaterialParameters001}
   &ScaleColor(
    Scale Alpha=&MakeCustomFloatFromBool(
     Bool=&Boolean_OrOperation(
      Bool1=&Boolean_AndOperation(Bool1=$System.bIsRunning, Bool2=&Boolean_NotOperation(Bool=$User.IsSkydiving))
      Bool2=$System.DEBUGSHOWTRAIL
     )
     True Float=1
     False Float=0
    )
    Scale RGB=&MakeVectorFromLinearColorRGB(Color=$System.TODColor)
    ScaleA=true
    ScaleRGB=true
   ) {module=ScaleColor001}
   &ScaleRibbonWidth(Ribbon Width Scale=&FloatFromCurve(
     FloatCurve=DataInterface() {propertyValues {Curve {keys [<at 0 1 RCIM_Cubic>, {arriveTangent=-0.6839, tangentMode=RCTM_User, time=1, value=0.2943}]}}}
     Scale Curve=&MergeHelper_FloatSwitch(
      0=&MakeFloatFromVector(Vector=$User.ShoeScale_L, Channel=::NewEnumerator0)
      1=&MakeFloatFromVector(Vector=$User.ShoeScale_R, Channel=::NewEnumerator0)
      2=?
      3=?
      Index=$Particles.SpawnGroup
      Num=::NewEnumerator0
     )
    ))
  ]
  rendererClasses [&NiagaraRibbonRendererProperties]
  renderers [{rendererClass=&NiagaraRibbonRendererProperties}]
 }
}
summary {
 systemName=NS_Shoe_BubbleBowl_Idle
 emitters [
  {emitterName=Trail, rendererClasses [&NiagaraRibbonRendererProperties]}
  {emitterName=Fish, rendererClasses [&NiagaraSpriteRendererProperties, &NiagaraMeshRendererProperties]}
 ]
 userVariables [
  User.bIsFrontEnd {type=NiagaraBool}
  User.bIsFrontEndPreview {type=NiagaraBool}
  User.bisPrimaryPreviewItem {type=NiagaraBool}
  User.bIsSwinging {type=NiagaraBool}
  User.FrontFishOffsetLerpX {type=NiagaraFloat, defaultValue=0, description=""}
  User.FrontFishOffsetLerpY {type=NiagaraFloat, defaultValue=0, description=""}
  User.HideFootFX {type=NiagaraBool}
  User.IsSkydiving {type=NiagaraBool}
  User.RearFishOffsetLerpZ {type=NiagaraFloat, defaultValue=0, description=""}
  User.RibbonIDOffset {type=NiagaraInt32}
  User.ShoeScale_L {type=Vector3f, description="", defaultValue {struct=Vector3f, value=<1 1 1>}}
  User.ShoeScale_R {type=Vector3f, description="", defaultValue {struct=Vector3f, value=<1 1 1>}}
 ]
}
