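# Parent material for the fish in the BubbleBowl cosmetic shoe, usable on skeletal meshes, mesh deformers and Niagara mesh particles.
# It uses material attributes with Default Lit shading, built as a chain of three material functions.
# The shared 2023 character PreFX function feeds MF_BubbleBowl_Fish, whose result enters the matching PostFX function.
# PreFX also supplies the auxiliary PostFX inputs directly, bypassing the fish function.
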Material /CosmeticShoes/Assets/FORT_Shoes/Shoe_BubbleBowl/Materials/M_BubbleBowl_Fish_Parent
graph {
 properties {
  bAutomaticallySetUsageInEditor=false
  bUseMaterialAttributes=true
  bUsedWithMeshDeformer=true
  bUsedWithNiagaraMeshParticles=true
  bUsedWithSkeletalMesh=true
  usedShadingModels=Default Lit
 }
 nodes [
  n1 = MaterialFunctionCall() {materialFunction=&MF_CharacterParent_2023_PreFX}
  MaterialFunctionCall(
   Material=MaterialFunctionCall!(In=@n1.Result) {materialFunction=&/MF_BubbleBowl_Fish}.Result
   M=@n1.M
   Pre Skinned Local Position=@n1.M
   Distance Blend=@n1.M
   Fresnel=@n1.M
  ) {materialFunction=&MF_CharacterParent_2023_PostFX}
 ]
}
