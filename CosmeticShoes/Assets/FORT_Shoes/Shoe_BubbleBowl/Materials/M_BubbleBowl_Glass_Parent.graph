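# Translucent parent glass material for the BubbleBowl cosmetic shoe, built on the shared MF_Glass function with a Fresnel-driven thin-film layer, both fed by TechArtMask.
# Static switches add camera dither and choose between collapsing vertices toward the object pivot via World Position Offset or applying character opacity.
# Vertex color blue scales opacity, and a selectable mask channel blends refraction from 1 to BB_Refraction.
# The result passes through MF_Sock, then MF_TechArtOverrides.
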Material /CosmeticShoes/Assets/FORT_Shoes/Shoe_BubbleBowl/Materials/M_BubbleBowl_Glass_Parent
graph {
 properties {
  bEnableResponsiveAA=true
  bScreenSpaceReflections=true
  bUseMaterialAttributes=true
  bUsedWithClothing=true
  bUsedWithMeshDeformer=true
  bUsedWithMorphTargets=true
  bUsedWithNanite=true
  bUsedWithNiagaraMeshParticles=true
  bUsedWithSkeletalMesh=true
  blendMode=BLEND_Translucent
  ditherOpacityMask=true
  pixelDepthOffsetMode=PDOM_Legacy
  refractionMethod=RM_PixelNormalOffset
  translucencyDirectionalLightingIntensity=5
  translucencyLightingMode=TLM_SurfacePerPixelLighting
  usedShadingModels=Default Lit
 }
 nodes [
  Constant() {r=1}
  n1 = TextureSampleParameter2D() {
   group=00_Glass
   parameterName=TechArtMask
   samplerType=SAMPLERTYPE_Masks
   sortPriority=-96
   texture=&T_BlackMask
  }
  n2 = MaterialFunctionCall!(
   Material Attributes=MaterialFunctionCall!(TechArtMask=@n1.RGB) {materialFunction=/Game/Global/Materials/Character_Parent_2023/Glass_Parent_2023/MF_Glass}.Result
   Fresnel=Fresnel() {baseReflectFraction=0, exponent=1}
   Alpha=@n1.RGB
  ) {materialFunction=/Game/Global/MaterialFunctions/TechArt/MF_ThinFilm_TechArt}
  n3 = StaticSwitchParameter(
   True=MaterialFunctionCall!(In=@n2.Result) {materialFunction=/Game/Packages/Fortress_SharedMaterials/Base_Material_Functions/BaseCustomizationFunctions/MF_FN_Vehicle_CameraDither}.Result
   False=@n2.Result
  ) {group=00_Glass, parameterName=UseCameraDither, sortPriority=250}
  n4 = Multiply(A=Subtract(A=ObjectPositionWS(), B=WorldPosition().XYZ), B=ScalarParameter() {group=00_Glass, parameterName=VertexHide, sortPriority=301})
  n5 = GetMaterialAttributes(None=StaticSwitchParameter(
    True=SetMaterialAttributes!(MaterialAttributes=@n3, World Position Offset=@n4) {attributeSetTypes [#1]}
    False=MaterialFunctionCall!(Material=@n3) {materialFunction=/Game/Global/Materials/Character_Parent_2023/Functions/BaseFN_CharacterOpacity_2023}.Result
   ) {group=00_Glass, parameterName=UseVertexHide, sortPriority=300}) {attributeGetTypes [#2]}
  n6 = Multiply(A=@n5.Opacity, B=VertexColor().B)
  n7 = LinearInterpolate(
   B=ScalarParameter() {defaultValue=1, group=BubbleBowl, parameterName=BB_Refraction}
   Alpha=ChannelMaskParameter(None=@n1.RGB) {group=BubbleBowl, parameterName=BB_RefractionChannel}.R
  ) {constA=1}
  MaterialFunctionCall(In=MaterialFunctionCall!(
    InOpacity=@n6
    Material=SetMaterialAttributes!(MaterialAttributes=@n5.MaterialAttributes, Refraction=@n7, Opacity=@n6, Emissive Color) {attributeSetTypes [#3, #2, #4]}
    Mask
   ) {materialFunction=/Game/Global/Materials/Cosmetics/Kicks/Socks/Materials/MF_Sock}.Result) {materialFunction=&MF_TechArtOverrides}
 ]
}
