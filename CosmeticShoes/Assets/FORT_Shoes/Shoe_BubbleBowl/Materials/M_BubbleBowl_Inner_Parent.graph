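# Masked two-sided parent material for the inner part of the BubbleBowl cosmetic shoe, wrapping the character PreFX and PostFX functions.
# TwoSidedSign swaps back faces to a tinted base color with emissive caustics, while front faces keep the parent look.
# A channel-selected mask cuts out front faces with dithered opacity, leaving back faces solid.
# A camera-offset WPO, limited by a height smoothstep, shifts the sole area.
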Material /CosmeticShoes/Assets/FORT_Shoes/Shoe_BubbleBowl/Materials/M_BubbleBowl_Inner_Parent
graph {
 properties {
  bEnableResponsiveAA=true
  bScreenSpaceReflections=true
  bUseMaterialAttributes=true
  bUsedWithClothing=true
  bUsedWithMeshDeformer=true
  bUsedWithMorphTargets=true
  bUsedWithNanite=true
  bUsedWithNiagaraMeshParticles=true
  bUsedWithSkeletalMesh=true
  blendMode=BLEND_Masked
  ditherOpacityMask=true
  pixelDepthOffsetMode=PDOM_Legacy
  translucencyDirectionalLightingIntensity=5
  translucencyLightingMode=TLM_SurfacePerPixelLighting
  twoSided=true
  usedShadingModels=Default Lit
 }
 nodes [
  n1 = MaterialFunctionCall() {materialFunction=&MF_CharacterParent_2023_PreFX}
  n2 = GetMaterialAttributes(None=@n1.Result) {attributeGetTypes [#1, #2]}
  n3 = Saturate(None=TwoSidedSign())
  n4 = VectorParameter() {group=BubbleBowl, parameterName=BB_EmissiveColor, defaultValue=<rgba 0 0.0321 0.0348 1>}
  n5 = MaterialFunctionCall!(Pre Skinned Local Postition=NamedRerouteUsage()) {materialFunction=&/MF_SimpleCaustics}
  n6 = ChannelMaskParameter(None=NamedRerouteUsage()) {group=BubbleBowl, parameterName=BB_OpacityChannel}
  n7 = OneMinus(None=Saturate(None=Multiply(A=@n6.R) {constB=2}))
  n8 = OneMinus(None=TwoSidedSign())
  n9 = ScalarParameter() {defaultValue=2.5, group=BubbleBowl, parameterName=SoleBackfaceCamOffset}
  n10 = SmoothStep(Min=Constant() {r=1}, Value=NamedRerouteUsage()) {constMax=0, constMin=10}
  n11 = SetMaterialAttributes!(
   MaterialAttributes=@n2.MaterialAttributes
   Base Color=LinearInterpolate(
    A=VectorParameter() {group=BubbleBowl, parameterName=BB_BackfaceBasecolor, defaultValue=<rgba 0.0066 0.5005 0.6867 1>}.RGB
    B=@n2.Base Color
    Alpha=@n3
   )
   Emissive Color=LinearInterpolate(A=Add(A=@n4.RGB, B=@n5.CausticsColor), B=@n2.Emissive Color, Alpha=@n3)
   Opacity Mask=Add(A=Multiply(A=@n3, B=@n7), B=Saturate(None=@n8))
   World Position Offset=Multiply(
    A=MaterialFunctionCall(Offset Amount=@n9) {materialFunction=/Engine/Functions/Engine_MaterialFunctions02/WorldPositionOffset/CameraOffset}.Camera Offset
    B=ComponentMask(None=@n10) {b=true}
   )
  ) {attributeSetTypes [
    #1
    #2
    #3
    #4
   ]}
  MaterialFunctionCall(Material=@n11, M=@n1.M, Pre Skinned Local Position=@n1.M, Distance Blend=@n1.M, Fresnel=@n1.M) {materialFunction=&MF_CharacterParent_2023_PostFX}
  NamedRerouteDeclaration(None=@n1.Pre skinned Local Position) {name=PreskinnedLocalPosition, nodeColor=<rgba 1 0.8672 0 1>}
  NamedRerouteDeclaration(None=@n1.Tech Art Mask) {name=TechArtMask, nodeColor=<rgba 1 0.5625 0 1>}
 ]
}
