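# Parent material for the BubbleBowl cosmetic shoe, adding animated-water style caustics to the standard character shading.
# The 2023 character PreFX function feeds a shoe reactivity function, then a sock function masked by the Tech Art Mask, then the PostFX function.
# MF_SimpleCaustics, driven by pre-skinned local position, is added to emissive.
# A reversed SmoothStep on local Z confines the caustics to the lower part of the shoe, fading out with height.
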Material /CosmeticShoes/Assets/FORT_Shoes/Shoe_BubbleBowl/Materials/M_BubbleBowl_Shoe_Parent
graph {
 properties {
  bUseMaterialAttributes=true
  bUsedWithClothing=true
  bUsedWithMeshDeformer=true
  bUsedWithMorphTargets=true
  bUsedWithNanite=true
  bUsedWithNiagaraMeshParticles=true
  bUsedWithSkeletalMesh=true
  pixelDepthOffsetMode=PDOM_Legacy
  previewMesh=/CosmeticShoes/Assets/FORT_Shoes/Shoe_BubbleBowl/Meshes/BubbleBowl_Shoe
  shadingModel=MSM_FromMaterialExpression
  usedShadingModels=Default Lit | Subsurface | Clear Coat | Subsurface Profile | Cloth
 }
 nodes [
  n1 = MaterialFunctionCall() {materialFunction=&MF_CharacterParent_2023_PreFX}
  n2 = MaterialFunctionCall!(Material=@n1.Result) {materialFunction=/Game/Global/Materials/Cosmetics/Kicks/Shoes/Materials/MF_ShoeReactivityTest}
  n3 = GetMaterialAttributes(None=MaterialFunctionCall(
    Material=MaterialFunctionCall!(InOpacity, Material=@n2.Result, Mask=@n1.Tech Art Mask) {materialFunction=/Game/Global/Materials/Cosmetics/Kicks/Socks/Materials/MF_Sock}.Result
    M=@n1.M
    Pre Skinned Local Position=@n1.M
    Distance Blend=@n1.M
    Fresnel=@n1.M
   ) {materialFunction=&MF_CharacterParent_2023_PostFX}.Result) {attributeGetTypes [#1, #2]}
  n4 = SmoothStep(Value=ComponentMask(None=@n1.Pre skinned Local Position) {b=true}) {constMax=3, constMin=6}
  n5 = MaterialFunctionCall!(Pre Skinned Local Postition=@n1.Pre skinned Local Position) {materialFunction=&/MF_SimpleCaustics}
  SetMaterialAttributes(MaterialAttributes=@n3.MaterialAttributes, Emissive Color=Add(A=@n3.Emissive Color, B=Multiply(A=@n4, B=@n5.CausticsColor))) {attributeSetTypes [#1]}
 ]
}
